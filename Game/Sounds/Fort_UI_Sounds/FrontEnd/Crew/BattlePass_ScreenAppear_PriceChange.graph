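# UI sound cue for the front-end Crew screens, named for a Battle Pass screen appear or price change, built from layered Battle Pass toolkit whooshes.
# A Mixer combines an immediate medium whoosh, a slightly delayed random pick of three medium whooshes, and a later, boosted surround whoosh with slight random upward pitch.
# The stack plays through the Fort_UI sound class at a heavily reduced overall volume.
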SoundCue /Game/Sounds/Fort_UI_Sounds/FrontEnd/Crew/BattlePass_ScreenAppear_PriceChange
root=Mixer_0
nodes [
 {id=Delay_0, children [Modulator_2], properties {delayMax=0.1, delayMin=0.1}}
 {id=Delay_1, children [Modulator_0], properties {delayMax=0.3, delayMin=0.3}}
 {id=Mixer_0, children [Delay_0,Delay_1,WavePlayer_2], properties {inputVolume [0.8,1.6,1]}}
 {id=Modulator_0, children [WavePlayer_0], properties {pitchMax=1.1, pitchMin=1, volumeMax=1, volumeMin=1}}
 {id=Modulator_2, children [Random_0], properties {pitchMax=1, pitchMin=1}}
 {id=Random_0, children [WavePlayer_1,WavePlayer_3,WavePlayer_4], properties {weights [1,1,1]}}
 {id=WavePlayer_0, properties {soundWaveAssetPtr=/Game/Sounds/Fort_UI_Sounds/FrontEnd/BattlePass/BP_ToolKit/BP_Prompt_Whoosh_Surround_01}}
 {id=WavePlayer_1, properties {soundWaveAssetPtr=/Game/Sounds/Fort_UI_Sounds/FrontEnd/BattlePass/BP_ToolKit/BP_Modal_Whoosh_Med_01}}
 {id=WavePlayer_2, properties {soundWaveAssetPtr=/Game/Sounds/Fort_UI_Sounds/FrontEnd/BattlePass/BP_ToolKit/BP_Modal_Whoosh_Med_03}}
 {id=WavePlayer_3, properties {soundWaveAssetPtr=/Game/Sounds/Fort_UI_Sounds/FrontEnd/BattlePass/BP_ToolKit/BP_Modal_Whoosh_Med_02}}
 {id=WavePlayer_4, properties {soundWaveAssetPtr=&BP_Modal_Whoosh_Med_03}}
]
properties {maxDistance=2097152, soundClassObject=&Fort_UI, volumeMultiplier=0.08}
