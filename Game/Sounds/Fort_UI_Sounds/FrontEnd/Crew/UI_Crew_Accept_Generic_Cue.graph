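# Sound cue for the generic accept sound in the front-end Crew UI, built from reused Battle Pass samples.
# A mixer layers a pitched-up button sound over a quieter surround whoosh with slight random upward pitch variation.
# A concurrency override limits it to one instance with a short retrigger time.
# It is assigned to the gameplay music sound class rather than a UI one.
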SoundCue /Game/Sounds/Fort_UI_Sounds/FrontEnd/Crew/UI_Crew_Accept_Generic_Cue
root=Mixer_0
nodes [
 {id=Mixer_0, children [Modulator_1,Modulator_0], properties {inputVolume [0.6,0.3]}}
 {id=Modulator_0, children [WavePlayer_0], properties {pitchMax=1.1, pitchMin=1, volumeMax=0.2, volumeMin=0.2}}
 {id=Modulator_1, children [WavePlayer_1], properties {pitchMax=1.5, pitchMin=1.5}}
 {id=WavePlayer_0, properties {soundWaveAssetPtr=/Game/Sounds/Fort_UI_Sounds/FrontEnd/BattlePass/BP_ToolKit/BP_Prompt_Whoosh_Surround_01}}
 {id=WavePlayer_1, properties {soundWaveAssetPtr=/Game/Sounds/Fort_UI_Sounds/FrontEnd/BattlePass/BP_Button_Generic_01}}
]
properties {
 bOverrideConcurrency=true
 maxDistance=2097152
 soundClassObject=/Game/Sounds/FortSoundClassesAndModes/Fort_Music_Gameplay
 volumeMultiplier=1
 concurrencyOverrides {maxCount=1, retriggerTime=0.3}
}
