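# UI sound cue in the Fort_UI sound class for a generic back action in the front-end Crew screens.
# A Mixer layers a main button click at full input volume with a Battle Pass back whoosh at a reduced input.
# The whoosh also passes through a Modulator with a fixed low volume, so it sits quietly beneath the click.
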SoundCue /Game/Sounds/Fort_UI_Sounds/FrontEnd/Crew/UI_Crew_Back_Generic_Cue
root=Mixer_0
nodes [
 {id=Mixer_0, children [WavePlayer_3,Modulator_3], properties {inputVolume [1,0.6]}}
 {id=Modulator_3, children [WavePlayer_1], properties {volumeMax=0.2, volumeMin=0.2}}
 {id=WavePlayer_1, properties {soundWaveAssetPtr=/Game/Sounds/Fort_UI_Sounds/FrontEnd/BattlePass/BP_Back_Whoosh_Surround_01}}
 {id=WavePlayer_3, properties {soundWaveAssetPtr=/Game/Sounds/Fort_UI_Sounds/FrontEnd/New_FrontEnd/UI_FN_Button_Click_Main_01}}
]
properties {maxDistance=2097152, soundClassObject=&Fort_UI, volumeMultiplier=1}
