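# Sound cue for the generic hover sound in the front-end Crew UI, routed to the Fort_UI sound class.
# A single wave player supplies the button hover click wave to a modulator that slightly randomizes pitch while keeping volume fixed.
# A concurrency override allows one instance at a time and stops the oldest, so rapid retriggers replace rather than stack.
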SoundCue /Game/Sounds/Fort_UI_Sounds/FrontEnd/Crew/UI_Crew_Hover_Generic_Cue
root=Modulator_0
nodes [
 {id=Modulator_0, children [WavePlayer_2], properties {pitchMax=1.03, pitchMin=0.97, volumeMax=1, volumeMin=1}}
 {id=WavePlayer_2, properties {soundWaveAssetPtr=/Game/Sounds/Fort_UI_Sounds/FrontEnd/New_FrontEnd/UI_FN_Button_Hover_Click_01}}
]
properties {
 bOverrideConcurrency=true
 maxDistance=2097152
 soundClassObject=&Fort_UI
 volumeMultiplier=1.1
 concurrencyOverrides {maxCount=1, resolutionRule=StopOldest}
}
