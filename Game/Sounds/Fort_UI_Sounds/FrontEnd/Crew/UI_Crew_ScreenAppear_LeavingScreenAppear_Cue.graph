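# Sound cue for a front-end Crew UI screen appearing, named for the leaving-screen appearance, routed to the Fort_UI sound class at reduced volume.
# A mixer layers five whoosh, swoosh and tile-slide waves, mostly borrowed from the BattlePass toolkit.
# The generic window whoosh and right-to-left menu swoosh start immediately, while fixed staggered delays bring in the tile slide, medium modal whoosh and follow-up whoosh in that order.
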SoundCue /Game/Sounds/Fort_UI_Sounds/FrontEnd/Crew/UI_Crew_ScreenAppear_LeavingScreenAppear_Cue
root=Mixer_0
nodes [
 {id=Delay_0, children [WavePlayer_0], properties {delayMax=0.3, delayMin=0.3}}
 {id=Delay_1, children [WavePlayer_1], properties {delayMax=0.4, delayMin=0.4}}
 {id=Delay_2, children [WavePlayer_4], properties {delayMax=0.2, delayMin=0.2}}
 {id=Mixer_0, children [WavePlayer_2,Delay_0,Delay_1,WavePlayer_3,Delay_2], properties {inputVolume [1,1,1,1,1]}}
 {id=WavePlayer_0, properties {soundWaveAssetPtr=/Game/Sounds/Fort_UI_Sounds/FrontEnd/BattlePass/BP_ToolKit/BP_Modal_Whoosh_Med_02}}
 {id=WavePlayer_1, properties {soundWaveAssetPtr=/Game/Sounds/Fort_UI_Sounds/FrontEnd/BattlePass/BP_ToolKit/BP_Modal_Whoosh_FollowUp_01}}
 {id=WavePlayer_2, properties {soundWaveAssetPtr=/Game/Sounds/Fort_UI_Sounds/FrontEnd/New_FrontEnd/Fortnite_GenericWindowAppear_Whoosh}}
 {id=WavePlayer_3, properties {soundWaveAssetPtr=/Game/Sounds/Fort_UI_Sounds/FrontEnd/BattlePass/BP_ToolKit/BP_MenuSwoosh_R_to_L}}
 {
  id=WavePlayer_4
  properties {soundWaveAssetPtr=/Game/Sounds/Fort_UI_Sounds/FrontEnd/BattlePass/BP_ToolKit/Fortnite_UI_BP_LandingPg_TileSlide}
 }
]
properties {maxDistance=2097152, soundClassObject=&Fort_UI, volumeMultiplier=0.2}
